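# Idle Niagara effect for the RankedFirth pickaxe, layering three sprite emitters: ray glow, core glow and sparkles.
# Diamond_Rays_Glow spawns continuously on the user skeletal mesh and alone uses Opacity_Scale.
# Glow is a burst sprite that tracks the mesh without expiring, feeding its sampled rotation axis into dynamic material parameters.
# Sparkle picks spawn points from a baked position array instead of sampling the mesh, with time-of-day tinting and two-way spin.
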NiagaraSystem /BRCosmetics/Weapons/FORT_Melee/Pickaxe_RankedFirth/FX/NS_Pickaxe_RankedFirth_Idle
dependencies {
 usedDataInterfaces [
  &NiagaraDataInterfaceArrayFloat3
  &NiagaraDataInterfaceCurve
  &NiagaraDataInterfaceArrayFloat
  &NiagaraDataInterfaceVectorCurve
 ]
 usedDynamicInputs [&Multiply_Vector2D_ByFloat]
 usedModules [&SystemState]
}
emitters {
 Diamond_Rays_Glow {
  emitterUpdateScript [
   &EmitterState(Inactive Response=::NewEnumerator1, Life Cycle Mode=::NewEnumerator1, Loop Duration=1, UseLoopDelay=false)
   &SpawnRate(SpawnRate=3, Use Spawn Probability=false)
  ]
  particleSpawnScript [
   &InitializeParticle(
    Color=<rgba 0.3563 0.3958 0.5 0.2>
    Sprite Rotation Mode=::NewEnumerator2
    Sprite Size Mode=::NewEnumerator3
    Uniform Sprite Size=60
    UsePositionOffset=true
    Write Material Random=true
   )
   &CameraOffset(Camera Offset Amount=-50)
  ]
  particleUpdateScript [
   &ParticleState()
   &SkeletalMeshLocation(MaskSpawnGroup=false, Mesh Sampling Type=::NewEnumerator5, Skeletal Mesh=$User.Skeletal Mesh, Socket Sampling Mode=::NewEnumerator5)
   &ScaleSpriteSize(Uniform Curve Sprite Scale=DataInterface() {propertyValues {Curve {keys [
        {arriveTangent=1.3062, interpMode=RCIM_Cubic, leaveTangent=1.3062, tangentMode=RCTM_User, time=0, value=0.5}
        {arriveTangent=0.1128, interpMode=RCIM_Cubic, leaveTangent=0.1128, tangentMode=RCTM_User, time=1, value=1}
       ]}}})
   &ScaleColor(
    Scale Alpha=&Multiply_Float(
     A=$User.Opacity_Scale
     B=&FloatFromCurve(
      FloatCurve=DataInterface() {propertyValues {Curve {keys [
          {arriveTangent=2.8155, interpMode=RCIM_Cubic, leaveTangent=2.8155, tangentMode=RCTM_User, time=0, value=0}
          {arriveTangent=0.0544, interpMode=RCIM_Cubic, leaveTangent=0.0544, tangentMode=RCTM_User, time=0.5, value=0.9847}
          {arriveTangent=-0.8319, interpMode=RCIM_Cubic, leaveTangent=-0.8319, tangentMode=RCTM_User, time=1, value=0}
         ]}}}
      Scale Curve=0.5
     )
    )
    Scale RGB=&VectorFromCurve(VectorCurve=DataInterface() {propertyValues {XCurve {keys [<at 0 1>,<at 1 1>]}, YCurve {keys [<at 0 1>,<at 1 0.75>]}, ZCurve {keys [<at 0 1>,<at 1 0.75>]}}})
    ScaleA=true
    ScaleRGB=true
   )
  ]
  rendererClasses [&NiagaraSpriteRendererProperties]
  renderers [{rendererClass=&NiagaraSpriteRendererProperties}]
 }
 Glow {
  emitterUpdateScript [
   &EmitterState(
    Inactive Response=::NewEnumerator1
    Life Cycle Mode=::NewEnumerator1
    Loop Behavior=::NewEnumerator1
    Loop Duration Mode=::NewEnumerator1
    UseLoopDelay=false
   )
   &SpawnBurst_Instantaneous(Use Spawn Probability=false, UseLoopCountLimit=false)
  ]
  particleSpawnScript [
   &InitializeParticle(
    Color=<rgba 0.4689 0.6298 1 0.8>
    Lifetime=1.5
    Position Offset=<0 -10 0>
    Sprite Size Mode=::NewEnumerator3
    Uniform Sprite Size=30
    UsePositionOffset=true
    Write Material Random=true
   )
   &CameraOffset(Camera Offset Amount=4)
  ]
  particleUpdateScript [
   &SkeletalMeshLocation(MaskSpawnGroup=false, Mesh Sampling Type=::NewEnumerator5, Skeletal Mesh=$User.Skeletal Mesh, Socket Sampling Mode=::NewEnumerator5)
   &ParticleState(Kill Particles When Lifetime Has Elapsed=false)
   &ScaleColor(Scale Alpha=0.5, ScaleA=true, ScaleRGB=true)
   &DynamicMaterialParameters(
    Index 0 Param 1=&MakeFloatFromVector(
     Vector=&RotateVectorByQuaternion(Vector To Rotate=<0 0 1>, Quaternion=$Particles.SkeletalMeshLocation.SampledRotation)
     Channel=::NewEnumerator0
    )
    Index 0 Param 2=&MakeFloatFromVector(
     Vector=&RotateVectorByQuaternion(Vector To Rotate=<0 0 1>, Quaternion=$Particles.SkeletalMeshLocation.SampledRotation)
     Channel=::NewEnumerator1
    )
    Index 0 Param 3=&MakeFloatFromVector(
     Vector=&RotateVectorByQuaternion(Vector To Rotate=<0 0 1>, Quaternion=$Particles.SkeletalMeshLocation.SampledRotation)
     Channel=::NewEnumerator2
    )
    Index 0 Param 4=50
    Index 1 Param 2=30
    Module.Index0Param0Write=true
    Module.Index0Param1Write=true
    Module.Index0Param2Write=true
    Module.Index0Param3Write=true
    Module.Index1Param0Write=true
    Module.Index1Param1Write=true
    Module.Index1Param2Write=true
    Module.Index1Param3Write=true
    Write Parameter Index 1=true
   )
  ]
  rendererClasses [&NiagaraSpriteRendererProperties]
  renderers [{rendererClass=&NiagaraSpriteRendererProperties}]
 }
 Sparkle {
  emitterUpdateScript [
   &EmitterState()
   &SpawnRate(Spawn Probability=0.5, SpawnRate=15, Use Spawn Probability=true)
  ]
  particleSpawnScript [
   &InitializeParticle(
    Color=&FNTimeOfDayColor(
     Night Color=<rgba 1.5 1.2492 0.328 1>
     Morning Color=<rgba 1.5 0.6623 0.15 1>
     Day Color=<rgba 1.5 0.6623 0.15 1>
     Evening Color=<rgba 1.5 0.6623 0.15 1>
    )
    Lifetime=&RandomRangeFloat(Minimum=0.5, Maximum=1)
    Sprite Rotation Angle=0.25
    Sprite Rotation Mode=::NewEnumerator1
    Sprite Size Mode=::NewEnumerator1
    Uniform Sprite Size Max=12.5
    Uniform Sprite Size Min=7.5
    UsePositionOffset=false
    Write Material Random=true
   )
   /SetVariables(Particles.Position=&SelectVectorFromArray(
     Array Sampling Mode=::NewEnumerator0
     Direct Array Index=?
     Vector Selection Array=DataInterface() {propertyValues {FloatData [
        <0.2834 0.1102 67.1708>
        <4.2218 14.705 40.3049>
        <-6.7033 4.7642 48.5855>
        <2.6648 23.9535 74.1777>
        <-5.1027 13.1127 66.2541>
        <-3.5052 3.4691 64.0809>
        <0.4236 28.1845 76.7636>
        <-3.9299 4.8255 84.3024>
        <5.173 8.5234 73.4488>
        <-7.4735 6.0889 47.1863>
        <-3.2842 24.5446 72.2373>
        <-7.8792 7.2447 47.5768>
        <2.9298 7.1437 77.5695>
        <-4.7447 4.6308 83.5058>
        <5.6257 6.1477 84.7858>
        <7.6809 3.2598 77.4396>
        <-8.107 13.4228 38.7664>
        <7.9939 2.6581 70.8403>
        <2.312 -0.2576 68.7152>
        <3.653 3.0851 64.0245>
        <-6.0535 9.1487 67.2143>
        <-0.1473 0.9549 60.5699>
        <7.4323 4.3126 67.5877>
        <2.0661 26.2042 74.8691>
        <-0.6139 27.9524 74.5978>
        <-10.9862 -2.4381 43.4472>
        <0.2934 4.7175 78.8447>
        <-3.5941 4.8623 84.4272>
        <-4.2033 0.7824 53.8066>
        <7.7043 1.3517 50.832>
        <9.4413 4.3009 47.5068>
        <5.561 15.175 67.5145>
        <4.1129 -1.1934 71.6285>
        <6.511 1.3111 53.1878>
        <-5.8071 8.9237 52.1744>
        <2.6263 8.1315 76.6342>
        <5.0019 19.1584 69.8053>
        <-5.1506 3.3072 77.3191>
        <7.1042 6.1774 70.2341>
        <-7.3708 3.9548 48.1454>
        <-3.9661 4.5886 83.1971>
        <3.9162 3.6368 78.6692>
        <-0.9972 0.261 55.8487>
        <-7.3652 3.4611 72.022>
        <6.476 9.5697 69.7308>
        <-5.5726 6.8087 74.5461>
        <4.8735 20.2993 70.953>
        <2.5026 25.9549 73.0449>
        <-2.9055 2.5653 64.1037>
        <6.2842 7.4496 53.1499>
        <0.6524 27.4705 72.0365>
        <-7.1806 3.289 77.5615>
        <4.9557 19.0784 68.1647>
        <2.2406 24.5554 67.8557>
        <4.1376 0.401 52.4784>
        <2.7871 23.9151 67.3731>
        <3.6011 22.3493 73.0096>
        <0.4165 16.9402 76.1773>
        <2.8844 11.5382 58.2104>
        <1.4006 14.6215 43.5989>
        <-1.738 22.4537 75.0088>
        <-2.3367 12.3955 60.4755>
        <-3.9657 5.1873 86.0043>
        <-0.2286 25.8552 66.8446>
        <-5.0967 6.1291 84.5831>
        <-8.0824 4.6988 76.7028>
        <3.0494 7.3969 77.0369>
        <3.5608 5.7488 88.513>
        <-0.5932 -2.591 49.5852>
        <8.9991 7.6109 36.6203>
        <8.4535 0.6456 71.4833>
        <0.3332 7.9311 77.2279>
        <-0.1662 5.6154 78.9536>
        <5.4017 16.111 67.4305>
        <6.4893 15.7582 40.3664>
        <7.9319 2.3805 50.2395>
        <-7.1878 4.0543 70.1007>
        <-4.8217 0.1831 68.4841>
        <7.9634 12.6236 43.1025>
        <-6.1418 8.4712 66.9033>
        <-9.5299 1.7038 48.6969>
        <-5.1715 9.5896 54.674>
        <-9.7181 4.2843 47.2543>
        <-5.1069 15.8049 69.4064>
        <5.5528 -1.7012 49.2302>
        <2.4951 13.5273 46.462>
        <-3.874 1.1373 74.5271>
        <6.6583 8.7947 47.4841>
        <2.0749 0.2551 67.3453>
        <-2.7633 5.4065 85.9089>
        <7.7897 1.0066 72.7396>
        <-3.1041 12.6409 48.5773>
        <7.2124 1.442 68.6612>
        <-8.8889 1.8199 49.8106>
        <-0.5965 -2.5815 48.6806>
        <-9.4448 0.1311 48.0704>
        <-8.1363 5.7416 75.9986>
        <10.1312 -0.2983 48.1558>
        <-2.617 4.5749 79.716>
        <8.3002 7.8257 48.3226>
       ]}}
     Random Seed=0
    )) {isSetParameters=true}
   &CameraOffset(Camera Offset Amount=1.5)
  ]
  particleUpdateScript [
   &ParticleState()
   &ScaleSpriteSize(Uniform Curve Sprite Scale=DataInterface() {propertyValues {Curve {keys [
        {arriveTangent=4.7337, interpMode=RCIM_Cubic, leaveTangent=4.7337, tangentMode=RCTM_User, time=0, value=0}
        {interpMode=RCIM_Cubic, tangentMode=RCTM_User, time=0.5, value=1}
        {tangentMode=RCTM_User, time=1, value=0}
       ]}}})
   &ScaleColor(
    Scale Alpha=&FloatFromCurve(FloatCurve=DataInterface() {propertyValues {Curve {keys [
         {arriveTangent=4.5234, interpMode=RCIM_Cubic, leaveTangent=4.5234, tangentMode=RCTM_User, time=0, value=0}
         {interpMode=RCIM_Cubic, tangentMode=RCTM_User, time=0.5, value=1}
         {tangentMode=RCTM_User, time=1, value=0}
        ]}}})
    Scale RGB=&Multiply_VectorByFloat(
     Vector=&FN_Select_Vector_from_Time_Of_Day(Morning=<1 1 1>, Day=<1 1 1>, Evening=<1 1 1>, Night=<0.5 0.5 2>)
     Float=&MakeFloatFromVector(
      Vector=&VectorFromFloat(Value=&FNTimeOfDayScalar(Night Scale=0.5, Morning Scale=1, Evening Scale=0.8))
      Channel=::NewEnumerator0
     )
    )
    ScaleA=true
    ScaleRGB=true
   )
   &SpriteRotationRate(Rotation Rate=&FloatFromCurve(
     FloatCurve=DataInterface() {propertyValues {Curve {keys [
         {interpMode=RCIM_Cubic, tangentMode=RCTM_User, time=0.2, value=0}
         {arriveTangent=3.4687, interpMode=RCIM_Cubic, leaveTangent=3.4687, tangentMode=RCTM_User, time=1, value=1}
        ]}}}
     Scale Curve=&SelectFloatFromArray(
      Array Sampling Mode=::NewEnumerator0
      Direct Array Index=?
      Float Selection Array=DataInterface() {propertyValues {FloatData [-800,800]}}
      Random Seed=0
     )
    ))
  ]
  rendererClasses [&NiagaraSpriteRendererProperties]
  renderers [{rendererClass=&NiagaraSpriteRendererProperties}]
 }
}
summary {
 systemName=NS_Pickaxe_RankedFirth_Idle
 emitters [
  {emitterName=Sparkle, rendererClasses [&NiagaraSpriteRendererProperties]}
  {emitterName=Glow, rendererClasses [&NiagaraSpriteRendererProperties]}
  {emitterName=Diamond_Rays_Glow, rendererClasses [&NiagaraSpriteRendererProperties]}
 ]
 userVariables [
  User.bIsFrontEnd {type=NiagaraBool}
  User.bIsFrontEndPreview {type=NiagaraBool}
  User.Opacity_Scale {type=NiagaraFloat, defaultValue=1, description=""}
  User.Skeletal Mesh {
   type=NiagaraDataInterfaceSkeletalMesh
   defaultValue {
    struct=NiagaraExt_VariableValue_DataInterface
    value {dataInterface=&self:NiagaraDataInterfaceSkeletalMesh_0}
   }
  }
 ]
}
