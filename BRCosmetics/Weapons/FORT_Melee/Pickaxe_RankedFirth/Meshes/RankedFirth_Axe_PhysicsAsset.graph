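# Physics asset for the RankedFirth two-handed pickaxe axe mesh, a cosmetic melee weapon.
# It defines a single capsule body on the Melee_TwoHanded_Handle bone, offset and slightly tilted along the handle.
# It has no constraints, so the weapon simulates or collides as one rigid piece.
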PhysicsAsset /BRCosmetics/Weapons/FORT_Melee/Pickaxe_RankedFirth/Meshes/RankedFirth_Axe_PhysicsAsset
bodies {Melee_TwoHanded_Handle [{shapeType=Capsule, length=71.9221, radius=16.8527, center=<0.7893 2.3948 40.3898>, rotation=<rot 1.3051 -90 0>}]}
constraints []
